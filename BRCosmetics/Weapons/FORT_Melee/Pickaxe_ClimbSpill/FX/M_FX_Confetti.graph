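# Unlit, two-sided translucent material for confetti Niagara mesh particles on the ClimbSpill pickaxe.
# Sphere masks on each UV axis are stepped into an adjustable border, blending OutlineColor with particle color.
# The UseChecker? scalar blends to a tiled two-color checker tinted by particle color, with TwoSidedSign shading back faces darker.
# The result feeds the shared MF_FX_EmissiveAdjustments function.
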Material /BRCosmetics/Weapons/FORT_Melee/Pickaxe_ClimbSpill/FX/M_FX_Confetti
graph {
 properties {
  bAutomaticallySetUsageInEditor=false
  bUsedWithNiagaraMeshParticles=true
  blendMode=BLEND_Translucent
  shadingModel=MSM_Unlit
  twoSided=true
  usedShadingModels=Unlit
 }
 nodes [
  n1 = VectorParameter() {parameterName=OutlineColor}
  n2 = ParticleColor()
  n3 = TextureCoordinate()
  n4 = Constant() {r=0.5}
  n5 = Step(
   Y=SphereMask(A=ComponentMask(None=@n3) {g=true}, B=@n4, Radius=@n4) {hardnessPercent=0}
   X=ScalarParameter() {defaultValue=0.025, parameterName=Y_OutlinePosition}
  )
  n6 = Step(
   Y=SphereMask(A=ComponentMask(None=@n3) {r=true}, B=@n4, Radius=@n4) {hardnessPercent=0}
   X=ScalarParameter() {defaultValue=0.35, parameterName=X_OutlinePosition}
  )
  n7 = OneMinus(None=Saturate(None=Add(A=@n5, B=@n6)))
  n8 = TextureCoordinate()
  n9 = ScalarParameter() {defaultValue=1, parameterName=CheckerTile}
  n10 = Multiply(
   A=LinearInterpolate(
    A=VectorParameter() {parameterName=CheckerColorA}.RGB
    B=VectorParameter() {parameterName=CheckerColorB, defaultValue=<rgba 1 1 1 0>}.RGB
    Alpha=MaterialFunctionCall!(UVs=@n8, Tiling=@n9) {materialFunction=/Engine/ArtTools/RenderToTexture/MaterialFunctions/CheckerPattern}.Result
   )
   B=LinearInterpolate(Alpha=TwoSidedSign()) {constA=0.8}
  )
  n11 = ParticleColor()
  MaterialFunctionCall(Emissive=LinearInterpolate(A=LinearInterpolate(A=@n1.RGB, B=@n2.RGB, Alpha=@n7), B=Multiply(A=@n10, B=@n11.RGB), Alpha=ScalarParameter() {parameterName=UseChecker?})) {materialFunction=&MF_FX_EmissiveAdjustments}
 ]
}
