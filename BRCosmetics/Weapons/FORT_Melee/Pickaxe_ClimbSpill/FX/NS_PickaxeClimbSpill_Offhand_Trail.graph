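# Niagara swing trail for the offhand weapon of the ClimbSpill pickaxe, built from four ribbon emitters: Base, BlackTopTrail, Shape and SupportingTrails.
# Each spawns and updates short-lived ribbons through shared FN anim-trail modules that work from bone indices.
# Base takes the system main color, Shape is light grey, and the other two are black.
# Alpha fades by curves scaled by SwingGeneralFade, while Base and Shape also apply time-of-day tinting.
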NiagaraSystem /BRCosmetics/Weapons/FORT_Melee/Pickaxe_ClimbSpill/FX/NS_PickaxeClimbSpill_Offhand_Trail
dependencies {
 usedDataInterfaces [&NiagaraDataInterfaceCurve, &NiagaraDataInterfaceSocketReader, &NiagaraDataInterfaceColorCurve]
 usedDynamicInputs [&ColorFromCurve, &FNTimeOfDayScalar]
 usedModules [
  &self:SystemSpawnScript.SystemScriptSource.SystemScriptGraph.NiagaraNodeAssignment_2.SetVariables_#1
  &SystemState
  &self:SystemSpawnScript.SystemScriptSource.SystemScriptGraph.NiagaraNodeAssignment_5.SetVariables_#2
  &self:SystemSpawnScript.SystemScriptSource.SystemScriptGraph.NiagaraNodeAssignment_6.SetVariables_#3
  &self:SystemSpawnScript.SystemScriptSource.SystemScriptGraph.NiagaraNodeAssignment_0.SetVariables_#4
 ]
}
emitters {
 Base {
  emitterSpawnScript [&FN_RibbonIDOffsetUtility()]
  emitterUpdateScript [
   &EmitterState()
   &FN_SpawnAnimTrail()
  ]
  particleSpawnScript [
   &FN_InitializeAnimTrailBoneIndecies()
   &FN_InitializeAnimTrail(Color=$System.MainColor, Lifetime=0.16)
  ]
  particleUpdateScript [
   &FN_UpdateAnimTrail()
   &ScaleColor(Scale RGB=&VectorFromFloat(Value=$System.TODFloat), ScaleA=false, ScaleRGB=true) {module=ScaleColor001}
   &ScaleColor(
    Scale Alpha=&Multiply_Float(
     A=&FloatFromCurve(
      FloatCurve=DataInterface() {propertyValues {Curve {keys [{interpMode=RCIM_Cubic, leaveTangent=-4.0138, tangentMode=RCTM_Break, time=0, value=1}, <at 0.4 0>]}}}
      Scale Curve=$System.SwingGeneralFade
     )
     B=1
    )
    Scale RGB=&MakeVectorFromLinearColorRGB(Color=$System.TODColor)
    ScaleA=true
    ScaleRGB=true
   ) {module=ScaleColor002}
  ]
  rendererClasses [&NiagaraRibbonRendererProperties]
  renderers [
   {rendererClass=&NiagaraRibbonRendererProperties}
   {rendererClass=&NiagaraRibbonRendererProperties, rendererIndex=1}
  ]
 }
 BlackTopTrail {
  emitterSpawnScript [&FN_RibbonIDOffsetUtility()]
  emitterUpdateScript [
   &EmitterState()
   &FN_SpawnAnimTrail()
  ]
  particleSpawnScript [
   &FN_InitializeAnimTrailBoneIndecies()
   &FN_InitializeAnimTrail(Color=<rgba 0 0 0 1>, Lifetime=0.16)
  ]
  particleUpdateScript [
   &SolveForcesAndVelocity(Clamp Velocity=false, Limit Acceleration=false)
   &FN_UpdateAnimTrail(Locked Particle RibbonWidth=1) {module=FN_UpdateAnimTrail001}
   &ScaleColor(
    Scale Alpha=&FloatFromCurve(
     FloatCurve=DataInterface() {propertyValues {Curve {keys [{interpMode=RCIM_Cubic, leaveTangent=-4.0138, tangentMode=RCTM_Break, time=0, value=1}, <at 0.4 0>]}}}
     Scale Curve=$System.SwingGeneralFade
    )
    ScaleA=true
    ScaleRGB=true
   )
  ]
  rendererClasses [&NiagaraRibbonRendererProperties]
  renderers [{rendererClass=&NiagaraRibbonRendererProperties}]
 }
 Shape {
  emitterUpdateScript [
   &EmitterState()
   &FN_SpawnAnimTrail()
  ]
  particleSpawnScript [
   &FN_InitializeAnimTrailBoneIndecies()
   &FN_InitializeAnimTrail(
    Color=&MultiplyLinearColor(
     Linear Color=<rgba 0.8 0.8 0.8 1>
     Scale Factor=&MakeVector4FromVectorAndFloat("Vector (RGB)"=<1 1 1>, "Float (Alpha)"=0.6)
    )
    Lifetime=0.16
   )
  ]
  particleUpdateScript [
   &SolveForcesAndVelocity(Clamp Velocity=false, Limit Acceleration=false)
   &FN_UpdateAnimTrail() {module=FN_UpdateAnimTrail001}
   &ScaleColor(
    Scale Alpha=&Multiply_Float(
     A=&FloatFromCurve(
      FloatCurve=DataInterface() {propertyValues {Curve {keys [
          {interpMode=RCIM_Cubic, tangentMode=RCTM_User, time=0, value=1}
          {arriveTangent=-5.0049, interpMode=RCIM_Cubic, leaveTangent=-5.0049, tangentMode=RCTM_User, time=0.5, value=0}
         ]}}}
      Scale Curve=$System.SwingGeneralFade
     )
     B=$System.TODFloat
    )
    Scale RGB=&MakeVectorFromLinearColorRGB(Color=$System.TODColor)
    ScaleA=true
    ScaleRGB=true
   ) {module=ScaleColor002}
  ]
  rendererClasses [&NiagaraRibbonRendererProperties]
  renderers [{rendererClass=&NiagaraRibbonRendererProperties}]
 }
 SupportingTrails {
  emitterSpawnScript [&FN_RibbonIDOffsetUtility()]
  emitterUpdateScript [
   &EmitterState()
   &FN_SpawnAnimTrail()
  ]
  particleSpawnScript [
   &FN_InitializeAnimTrailBoneIndecies()
   &FN_InitializeAnimTrail(Color=<rgba 0 0 0 1>, Lifetime=0.16)
  ]
  particleUpdateScript [
   &SolveForcesAndVelocity(Clamp Velocity=false, Limit Acceleration=false)
   &FN_UpdateAnimTrail(Locked Particle RibbonWidth=1) {module=FN_UpdateAnimTrail001}
   &ScaleColor(
    Scale Alpha=&FloatFromCurve(
     FloatCurve=DataInterface() {propertyValues {Curve {keys [
         {interpMode=RCIM_Cubic, tangentMode=RCTM_User, time=0, value=1}
         {arriveTangent=-5.0049, interpMode=RCIM_Cubic, leaveTangent=-5.0049, tangentMode=RCTM_User, time=0.5, value=0}
        ]}}}
     Scale Curve=$System.SwingGeneralFade
    )
    ScaleA=true
    ScaleRGB=false
   )
  ]
  rendererClasses [&NiagaraRibbonRendererProperties]
  renderers [{rendererClass=&NiagaraRibbonRendererProperties}]
 }
}
summary {
 systemName=NS_PickaxeClimbSpill_Offhand_Trail
 emitters [
  {emitterName=SupportingTrails, rendererClasses [&NiagaraRibbonRendererProperties]}
  {emitterName=Base, rendererClasses [&NiagaraRibbonRendererProperties]}
  {emitterName=Shape, rendererClasses [&NiagaraRibbonRendererProperties]}
  {emitterName=BlackTopTrail, rendererClasses [&NiagaraRibbonRendererProperties]}
 ]
 userVariables [User.RibbonIDOffset {type=NiagaraInt32}]
}
